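# Physics asset for the right-arm mesh of the large male BasilStrong Mashup player body, covering two dangling bandana tufts, A and B.
# Each tuft is a three-bone chain of capsule bodies, with a small base capsule and two offset, rotated segments.
# Constraints link base to first segment to second, all sharing the same swing limit angles and a zero twist limit angle.
# No other arm bones have bodies, and the two chains share no constraint.
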PhysicsAsset /Game/Characters/Player/Male/Large/Bodies/M_LRG_BasilStrong_Mashup/Meshes/M_LRG_BasilStrong_Mashup_ArmRt_Physics
bodies {
 dyn_arm_r_bandana_tuftA_01 [{
   shapeType=Capsule
   length=2.024
   radius=1.5845
   shapeName=dyn_arm_r_bandana_tuftA_01_capsule
   center=<2.2987 -0.1737 -0.0114>
   rotation=<rot 0 93.009 -70.8426>
  }]
 dyn_arm_r_bandana_tuftA_02 [{
   shapeType=Capsule
   length=3.8121
   radius=1.4436
   shapeName=dyn_arm_r_bandana_tuftA_02_capsule
   center=<2.6639 0.1209 -0.5022>
   rotation=<rot 0 -83.6732 -80.411>
  }]
 dyn_arm_r_bandana_tuftA_base [{shapeType=Capsule, length=0.505, radius=0.505, shapeName=dyn_arm_r_bandana_tuftA_base_capsule, rotation=<rot 0 -90 0>}]
 dyn_arm_r_bandana_tuftB_01 [{
   shapeType=Capsule
   length=1.6377
   radius=0.9359
   shapeName=dyn_arm_r_bandana_tuftB_01_capsule
   center=<1.9758 0.16 0.0336>
   rotation=<rot 90 0 0>
  }]
 dyn_arm_r_bandana_tuftB_02 [{
   shapeType=Capsule
   length=3.7563
   radius=1.4302
   shapeName=dyn_arm_r_bandana_tuftB_02_capsule
   center=<3.5221 0.4606 0.4942>
   rotation=<rot 0 106.1579 -86.9322>
  }]
 dyn_arm_r_bandana_tuftB_base [{shapeType=Capsule, length=0.505, radius=0.505, shapeName=dyn_arm_r_bandana_tuftB_base_capsule, rotation=<rot 0 -90 0>}]
}
constraints [
 {
  bone1Name=dyn_arm_r_bandana_tuftA_02
  bone2Name=dyn_arm_r_bandana_tuftA_01
  swing1LimitDegrees=45
  swing2LimitDegrees=45
  twistLimitDegrees=0
 }
 {
  bone1Name=dyn_arm_r_bandana_tuftA_01
  bone2Name=dyn_arm_r_bandana_tuftA_base
  swing1LimitDegrees=45
  swing2LimitDegrees=45
  twistLimitDegrees=0
 }
 {
  bone1Name=dyn_arm_r_bandana_tuftB_01
  bone2Name=dyn_arm_r_bandana_tuftB_base
  swing1LimitDegrees=45
  swing2LimitDegrees=45
  twistLimitDegrees=0
 }
 {
  bone1Name=dyn_arm_r_bandana_tuftB_02
  bone2Name=dyn_arm_r_bandana_tuftB_01
  swing1LimitDegrees=45
  swing2LimitDegrees=45
  twistLimitDegrees=0
 }
]
